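# Sound cue for piano impact hits in the demolition impact set, classed under Fort_Impacts.
# One of six equally weighted piano hit waves is picked at random, slightly pitched up by a random amount, then passed through close-hit demolition crossfade attenuation.
# It also sends to the exterior ambient source bus.
# A shared concurrency rule, named for a four-voice, stop-furthest-then-oldest limit with volume scaling, governs overlapping hits.
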SoundCue /Game/Sounds/Fort_Impact_Sounds/Demolition/Piano/Spot_Piano_Impact_Cue
root=Attenuation_3
nodes [
 {
  id=Attenuation_3
  children [Modulator_3]
  properties {attenuationSettings=/Game/Sounds/Attenuation/Fort_Demolition_Crossfade_Global_Hit_Close_Att}
 }
 {id=Modulator_3, children [Random_1], properties {pitchMax=1.2, pitchMin=1}}
 {
  id=Random_1
  children [WavePlayer_6, WavePlayer_7, WavePlayer_8, WavePlayer_9, WavePlayer_10, WavePlayer_11]
  properties {weights [1,1,1,1,1,1]}
 }
 {id=WavePlayer_10, properties {soundWaveAssetPtr=&/Piano_Hits_02}}
 {id=WavePlayer_11, properties {soundWaveAssetPtr=&/Piano_Hits_01}}
 {id=WavePlayer_6, properties {soundWaveAssetPtr=&/Piano_Hits_06}}
 {id=WavePlayer_7, properties {soundWaveAssetPtr=&/Piano_Hits_05}}
 {id=WavePlayer_8, properties {soundWaveAssetPtr=&/Piano_Hits_04}}
 {id=WavePlayer_9, properties {soundWaveAssetPtr=&/Piano_Hits_03}}
]
properties {
 attenuationSettings=&Fort_Demolition_Crossfade_Global_Hit_Close_Att
 maxDistance=4800
 soundClassObject=&Fort_Impacts
 volumeMultiplier=1
 busSends [{
   audioBus=null
   maxSendDistance=6000
   minSendLevel=0.4
   sendLevel=1
   soundSourceBus=&AmbientAudio_Exterior_SourceBus
   customSendLevelCurve {editorCurveData {keys [<at 0 0.4>,<at 1 1>]}}
  }]
 concurrencyOverrides {resolutionRule=PreventNew}
 concurrencySet [/Game/Sounds/Concurrency_Rules/FourMaximum_StopFurthestOldest_VolumeScale]
}
