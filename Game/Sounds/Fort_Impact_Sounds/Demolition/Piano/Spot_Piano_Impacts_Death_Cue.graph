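# Sound cue for piano destruction impacts, played by name as the death sound when a piano is demolished.
# It picks one of six equally weighted Piano_Death waves and runs it through a modulator that randomizes pitch slightly below normal.
# Playback uses the Fort_Impacts class, shared demolition close-hit attenuation and a four-voice, stop-furthest-then-oldest concurrency rule.
# It also sends to the exterior ambient audio source bus.
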SoundCue /Game/Sounds/Fort_Impact_Sounds/Demolition/Piano/Spot_Piano_Impacts_Death_Cue
root=Modulator_1
nodes [
 {id=Modulator_1, children [Random_2], properties {pitchMax=0.9, pitchMin=0.8}}
 {
  id=Random_2
  children [WavePlayer_16, WavePlayer_17, WavePlayer_18, WavePlayer_19, WavePlayer_20, WavePlayer_21]
  properties {weights [1,1,1,1,1,1]}
 }
 {id=WavePlayer_16, properties {soundWaveAssetPtr=&/Piano_Death_01}}
 {id=WavePlayer_17, properties {soundWaveAssetPtr=&/Piano_Death_02}}
 {id=WavePlayer_18, properties {soundWaveAssetPtr=&/Piano_Death_03}}
 {id=WavePlayer_19, properties {soundWaveAssetPtr=&/Piano_Death_04}}
 {id=WavePlayer_20, properties {soundWaveAssetPtr=&/Piano_Death_05}}
 {id=WavePlayer_21, properties {soundWaveAssetPtr=&/Piano_Death_06}}
]
properties {
 attenuationSettings=/Game/Sounds/Attenuation/Fort_Demolition_Crossfade_Global_Hit_Close_Att
 maxDistance=4800
 soundClassObject=&Fort_Impacts
 volumeMultiplier=1.3
 attenuationOverrides {bAttenuateWithLPF=true, distanceAlgorithm=NaturalSound, attenuationShapeExtents {x=1000}}
 busSends [{
   audioBus=null
   maxSendDistance=6000
   minSendLevel=0.4
   sendLevel=1
   soundSourceBus=&AmbientAudio_Exterior_SourceBus
   customSendLevelCurve {editorCurveData {keys [<at 0 0.4>,<at 1 1>]}}
  }]
 concurrencyOverrides {resolutionRule=PreventNew}
 concurrencySet [/Game/Sounds/Concurrency_Rules/FourMaximum_StopFurthestOldest_VolumeScale]
}
